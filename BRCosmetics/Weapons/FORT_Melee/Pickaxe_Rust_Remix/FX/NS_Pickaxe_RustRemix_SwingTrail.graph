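# Niagara ribbon swing trail for the Rust Remix pickaxe cosmetic.
# One SwingTrail emitter spawns short-lived particles through shared FN anim-trail modules, using ribbon ID offset and bone index setup.
# A variant selector picks among tinted trail colors, each multiplied by a color curve indexed by the Fortnite time-of-day parameter to adapt tint and alpha.
# An alpha curve, scaled by a system swing fade value, fades the ribbon out.
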NiagaraSystem /BRCosmetics/Weapons/FORT_Melee/Pickaxe_Rust_Remix/FX/NS_Pickaxe_RustRemix_SwingTrail
dependencies {
 usedDataInterfaces [&NiagaraDataInterfaceColorCurve, &NiagaraDataInterfaceCurve, &NiagaraDataInterfaceSocketReader]
 usedDynamicInputs [&MergeHelper_FloatSwitch]
 usedModules [
  &self:SystemSpawnScript.SystemScriptSource.SystemScriptGraph.NiagaraNodeAssignment_1.SetVariables_#1
  &SystemState
  &FN_VariantManager
  &self:SystemSpawnScript.SystemScriptSource.SystemScriptGraph.NiagaraNodeAssignment_2.SetVariables_#2
  &self:SystemSpawnScript.SystemScriptSource.SystemScriptGraph.NiagaraNodeAssignment_3.SetVariables_#3
 ]
}
emitters {SwingTrail {
  emitterSpawnScript [&FN_RibbonIDOffsetUtility()]
  emitterUpdateScript [
   &EmitterState()
   &FN_SpawnAnimTrail()
  ]
  particleSpawnScript [
   &FN_InitializeAnimTrailBoneIndecies()
   &FN_InitializeAnimTrail(
    Color=&VariantColorSelector(
     Variant Amount=2
     Original Color=&MultiplyLinearColor(
      Linear Color=<rgba 0.0116 0.0123 0.0262 2.3>
      Scale Factor=&MakeVector4_FromLinearColor(LinearColor=&ColorFromCurve(
        ColorCurve=DataInterface() {propertyValues {
          AlphaCurve {keys [<at 0 1.3>,<at 0.125 1.2>,<at 0.25 1.1>,<at 0.5 1>,<at 0.75 1.1>,<at 0.875 1.2>,<at 1 1.3>]}
          BlueCurve {keys [<at 0.0417 1>,<at 0.125 0.8304>,<at 0.25 0.7>,<at 0.375 0.935>,<at 0.5 0.895>,<at 0.625 0.935>,<at 0.75 0.7>,<at 0.875 0.675>,<at 0.9167 1>,<at 1 1>]}
          GreenCurve {keys [<at 0.0417 0.5711>,<at 0.125 0.6609>,<at 0.25 0.731>,<at 0.375 0.935>,<at 0.5 0.895>,<at 0.625 0.935>,<at 0.75 0.7>,<at 0.875 0.7467>,<at 0.9167 0.6376>,<at 1 0.5711>]}
          RedCurve {keys [<at 0.0417 0.2664>,<at 0.125 1>,<at 0.25 1>,<at 0.375 0.935>,<at 0.5 0.895>,<at 0.625 0.935>,<at 0.75 1>,<at 0.875 1>,<at 0.9167 0.6>,<at 1 0.2664>]}
         }}
        CurveIndex=Linked() {linkedVariable {name=NPC.FortniteNPC.TimeOfDay_0-1, type=NiagaraFloat}}
       ))
     )
     Variant 1 Color=&MultiplyLinearColor(
      Linear Color=<rgba 0.25 0.0306 0.0179 2>
      Scale Factor=&MakeVector4_FromLinearColor(LinearColor=&ColorFromCurve(
        ColorCurve=DataInterface() {propertyValues {
          AlphaCurve {keys [<at 0 1.3>,<at 0.125 1.1>,<at 0.25 1>,<at 0.5 1>,<at 0.75 1>,<at 0.875 1.1>,<at 1 1.3>]}
          BlueCurve {keys [<at 0.0417 1>,<at 0.125 0.9625>,<at 0.25 0.95>,<at 0.375 0.935>,<at 0.5 0.895>,<at 0.625 0.935>,<at 0.75 0.95>,<at 0.875 0.925>,<at 0.9167 1>,<at 1 1>]}
          GreenCurve {keys [<at 0.0417 0.6054>,<at 0.125 0.925>,<at 0.25 0.9552>,<at 0.375 0.935>,<at 0.5 0.895>,<at 0.625 0.935>,<at 0.75 0.95>,<at 0.875 0.9415>,<at 0.9167 0.6376>,<at 1 0.6054>]}
          RedCurve {keys [<at 0.0417 0.325>,<at 0.125 1>,<at 0.25 1>,<at 0.375 0.935>,<at 0.5 0.895>,<at 0.625 0.935>,<at 0.75 1>,<at 0.875 1>,<at 0.9167 0.6>,<at 1 0.325>]}
         }}
        CurveIndex=Linked() {linkedVariable {name=NPC.FortniteNPC.TimeOfDay_0-1, type=NiagaraFloat}}
       ))
     )
     Variant 2 Color=&MultiplyLinearColor(
      Linear Color=<rgba 0.3 0.005 0.0064 1.65>
      Scale Factor=&MakeVector4_FromLinearColor(LinearColor=&ColorFromCurve(
        ColorCurve=DataInterface() {propertyValues {
          AlphaCurve {keys [<at 0 1.2>,<at 0.125 1.1>,<at 0.25 1>,<at 0.5 1>,<at 0.75 1>,<at 0.875 1.1>,<at 1 1.2>]}
          BlueCurve {keys [<at 0.0417 1>,<at 0.125 0.8304>,<at 0.25 0.7>,<at 0.375 0.935>,<at 0.5 0.895>,<at 0.625 0.935>,<at 0.75 0.7>,<at 0.875 0.675>,<at 0.9167 1>,<at 1 1>]}
          GreenCurve {keys [<at 0.0417 0.5711>,<at 0.125 0.6609>,<at 0.25 0.731>,<at 0.375 0.935>,<at 0.5 0.895>,<at 0.625 0.935>,<at 0.75 0.7>,<at 0.875 0.7467>,<at 0.9167 0.6376>,<at 1 0.5711>]}
          RedCurve {keys [<at 0.0417 0.2664>,<at 0.125 1>,<at 0.25 1>,<at 0.375 0.935>,<at 0.5 0.895>,<at 0.625 0.935>,<at 0.75 1>,<at 0.875 1>,<at 0.9167 0.6>,<at 1 0.2664>]}
         }}
        CurveIndex=Linked() {linkedVariable {name=NPC.FortniteNPC.TimeOfDay_0-1, type=NiagaraFloat}}
       ))
     )
     Variant 3 Color=&MultiplyLinearColor(
      Linear Color=<rgba 0.1906 0.3 0.0812 2>
      Scale Factor=&MakeVector4_FromLinearColor(LinearColor=&ColorFromCurve(
        ColorCurve=DataInterface() {propertyValues {
          AlphaCurve {keys [<at 0 1.3>,<at 0.125 1.2>,<at 0.25 1.1>,<at 0.5 1>,<at 0.75 1.1>,<at 0.875 1.2>,<at 1 1.3>]}
          BlueCurve {keys [<at 0.0417 1>,<at 0.125 0.9>,<at 0.25 0.825>,<at 0.375 0.935>,<at 0.5 0.895>,<at 0.625 0.935>,<at 0.75 0.825>,<at 0.875 0.8>,<at 0.9167 1>,<at 1 1>]}
          GreenCurve {keys [<at 0.0417 0.5711>,<at 0.125 0.8>,<at 0.25 0.8431>,<at 0.375 0.935>,<at 0.5 0.895>,<at 0.625 0.935>,<at 0.75 0.825>,<at 0.875 0.8441>,<at 0.9167 0.6376>,<at 1 0.5711>]}
          RedCurve {keys [<at 0.0417 0.2664>,<at 0.125 1>,<at 0.25 1>,<at 0.375 0.935>,<at 0.5 0.895>,<at 0.625 0.935>,<at 0.75 1>,<at 0.875 1>,<at 0.9167 0.6>,<at 1 0.2664>]}
         }}
        CurveIndex=Linked() {linkedVariable {name=NPC.FortniteNPC.TimeOfDay_0-1, type=NiagaraFloat}}
       ))
     )
    )
    Lifetime=0.3
   )
  ]
  particleUpdateScript [
   &SolveForcesAndVelocity(Clamp Velocity=false, Limit Acceleration=false)
   &FN_UpdateAnimTrail() {module=FN_UpdateAnimTrail001}
   &ScaleColor(
    Scale Alpha=&FloatFromCurve(
     FloatCurve=DataInterface() {propertyValues {Curve {keys [{arriveTangent=-14.0869, interpMode=RCIM_Cubic, time=0.3, value=1}, <at 1 0 RCIM_Cubic>]}}}
     Scale Curve=$System.SwingGeneralFade
    )
    Scale RGB=&MakeVectorFromLinearColorRGB(Color=<rgba 1 1 1 1>)
    ScaleA=true
    ScaleRGB=true
   )
  ]
  rendererClasses [&NiagaraRibbonRendererProperties]
  renderers [{rendererClass=&NiagaraRibbonRendererProperties}]
 }}
summary {
 systemName=NS_Pickaxe_RustRemix_SwingTrail
 emitters [{emitterName=SwingTrail, rendererClasses [&NiagaraRibbonRendererProperties]}]
 userVariables [
  User.RibbonIDOffset {type=NiagaraInt32}
  User.SwingSectionIndex {type=NiagaraInt32}
  User.Variant {type=NiagaraFloat, defaultValue=0, description=""}
 ]
}
