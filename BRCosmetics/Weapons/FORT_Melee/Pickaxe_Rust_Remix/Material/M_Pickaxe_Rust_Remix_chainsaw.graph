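# Masked, two-sided lit material for the chainsaw part of the Rust Remix pickaxe cosmetic, usable on skeletal meshes.
# A single Panner scrolls UVs along U at a rate set by the PanSpeed_chainsaw parameter.
# The diffuse, normal, mask and alpha textures all sample those shared UVs, so the surface detail and alpha pattern move together.
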Material /BRCosmetics/Weapons/FORT_Melee/Pickaxe_Rust_Remix/Material/M_Pickaxe_Rust_Remix_chainsaw
graph {
 properties {
  bUsedWithMeshDeformer=true
  bUsedWithSkeletalMesh=true
  blendMode=BLEND_Masked
  pixelDepthOffsetMode=PDOM_Legacy
  twoSided=true
  usedShadingModels=Default Lit
 }
 nodes [
  n1 = Panner(Coordinate=TextureCoordinate(), Speed=AppendVector(A=ScalarParameter() {defaultValue=0.4, parameterName=PanSpeed_chainsaw}, B=Constant()))
  TextureSample(UVs=@n1) {texture=/BRCosmetics/Weapons/FORT_Melee/Pickaxe_Rust_Remix/Textures/T_Rust_Remix_D}
  TextureSample(UVs=@n1) {samplerType=SAMPLERTYPE_Normal, texture=/BRCosmetics/Weapons/FORT_Melee/Pickaxe_Rust_Remix/Textures/T_Rust_Remix_N}
  TextureSample(UVs=@n1) {samplerType=SAMPLERTYPE_Masks, texture=/BRCosmetics/Weapons/FORT_Melee/Pickaxe_Rust_Remix/Textures/T_Rust_Remix_S}
  TextureSample(UVs=@n1) {samplerType=SAMPLERTYPE_Masks, texture=/BRCosmetics/Weapons/FORT_Melee/Pickaxe_Rust_Remix/Textures/Rust_Remix_Alpha}
 ]
}
